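# Niagara system for the Cosmic Squatter v02 pickaxe's animation trail, drawing short-lived ribbons from bone or socket pairs on a skeletal mesh.
# Two emitters, NebulaTrail and SmokeTrail, share FN anim-trail spawn, initialize and update modules, and each has two ribbon renderers.
# Nebula widens over particle life and fades by both particle and emitter age.
# Smoke widens with emitter age, shifts color through a curve and drives dynamic material parameters.
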NiagaraSystem /BRCosmetics/Weapons/FORT_Melee/Pickaxe_CosmicSquatter/FX/NS_Pickaxe_CosmicSquatter_v02_Trail
dependencies {
 usedDataInterfaces [&NiagaraDataInterfaceColorCurve, &NiagaraDataInterfaceCurve]
 usedDynamicInputs []
 usedModules [
  &self:SystemSpawnScript.SystemScriptSource.SystemScriptGraph.NiagaraNodeAssignment_0.SetVariables_#1
  &SystemState
 ]
}
emitters {
 NebulaTrail {
  emitterSpawnScript [&FN_RibbonIDOffsetUtility()]
  emitterUpdateScript [
   &EmitterState()
   &FN_SpawnAnimTrail()
  ]
  particleSpawnScript [
   &FN_InitializeAnimTrailBoneIndecies()
   &FN_InitializeAnimTrailBoneIndecies(AnimTrail Selection=1, First Bone or Socket Index=2, Second Bone or Socket Index=3) {module=FN_InitializeAnimTrailBoneIndecies001}
   &FN_InitializeAnimTrail(Lifetime=0.125, SkeletalMesh=$System.SkeletalMesh)
   &AddVelocity(Velocity=<0 400 0>) {enabled=false}
  ]
  particleUpdateScript [
   &SolveForcesAndVelocity(Clamp Velocity=false, Limit Acceleration=false)
   &FN_UpdateAnimTrail(Locked Particle RibbonWidth=1, SkeletalMesh=$System.SkeletalMesh) {module=FN_UpdateAnimTrail001}
   &ScaleColor(
    Scale Alpha=&FloatFromCurve(FloatCurve=DataInterface() {propertyValues {Curve {keys [
         {tangentMode=RCTM_User, time=0, value=1}
         {arriveTangent=-0.0405, interpMode=RCIM_Cubic, leaveTangent=-0.0405, tangentMode=RCTM_User, time=0.7009, value=1}
         {arriveTangent=-8.0089, interpMode=RCIM_Cubic, leaveTangent=-8.0089, tangentMode=RCTM_User, time=1, value=0}
        ]}}})
    ScaleA=true
    ScaleRGB=false
   )
   &ScaleColor(
    Scale Alpha=&FloatFromCurve(
     FloatCurve=DataInterface() {propertyValues {Curve {keys [
         {arriveTangent=-15.0997, interpMode=RCIM_Cubic, leaveTangent=-15.0997, tangentMode=RCTM_User, time=0.1, value=1}
         {arriveTangent=-0.1271, interpMode=RCIM_Cubic, leaveTangent=-0.1271, tangentMode=RCTM_User, time=0.23, value=0}
        ]}}}
     CurveIndex=$Emitter.Age
    )
    ScaleA=true
    ScaleRGB=false
   ) {module=ScaleColor001}
   &ScaleRibbonWidth(Ribbon Width Scale=&FloatFromCurve(FloatCurve=DataInterface() {propertyValues {Curve {keys [
         {arriveTangent=4.2888, interpMode=RCIM_Cubic, leaveTangent=4.2888, tangentMode=RCTM_User, time=0, value=2}
         {arriveTangent=0.6625, interpMode=RCIM_Cubic, leaveTangent=0.6625, tangentMode=RCTM_User, time=1, value=4.5}
        ]}}}))
  ]
  rendererClasses [&NiagaraRibbonRendererProperties]
  renderers [
   {rendererClass=&NiagaraRibbonRendererProperties}
   {rendererClass=&NiagaraRibbonRendererProperties, rendererIndex=1}
  ]
 }
 SmokeTrail {
  emitterSpawnScript [&FN_RibbonIDOffsetUtility()]
  emitterUpdateScript [
   &EmitterState()
   &FN_SpawnAnimTrail()
  ]
  particleSpawnScript [
   &FN_InitializeAnimTrailBoneIndecies()
   &FN_InitializeAnimTrailBoneIndecies(AnimTrail Selection=1, First Bone or Socket Index=2, Second Bone or Socket Index=3) {enabled=false, module=FN_InitializeAnimTrailBoneIndecies001}
   &FN_InitializeAnimTrail(AnimTrailWidth=0.7, Lifetime=0.125, SkeletalMesh=$System.SkeletalMesh)
   &AddVelocity(Velocity=<0 400 0>) {enabled=false}
  ]
  particleUpdateScript [
   &SolveForcesAndVelocity(Clamp Velocity=false, Limit Acceleration=false)
   &FN_UpdateAnimTrail(Locked Particle RibbonWidth=1, SkeletalMesh=$System.SkeletalMesh) {module=FN_UpdateAnimTrail001}
   &Color(
    Color=&ColorFromCurve(ColorCurve=DataInterface() {propertyValues {
       AlphaCurve {keys [<at 0 2 RCIM_Cubic>]}
       BlueCurve {keys [
         <at 0 5>
         {arriveTangent=-41.8914, interpMode=RCIM_Cubic, leaveTangent=-41.8914, time=0.0561, value=5}
         <at 0.1074 0.5>
         <at 0.1667 3>
         <at 1 4>
        ]}
       GreenCurve {keys [
         <at 0 0.1>
         {arriveTangent=40.0295, interpMode=RCIM_Cubic, leaveTangent=40.0295, time=0.0561, value=4.1667}
         <at 0.1074 4.4>
         <at 0.1667 0.15>
         <at 1 0.08>
        ]}
       RedCurve {keys [
         <at 0 0.1>
         {arriveTangent=45.615, interpMode=RCIM_Cubic, leaveTangent=45.615, time=0.0561, value=0}
         <at 0.1074 5>
         <at 0.1667 0.3875>
         <at 1 1.3867>
        ]}
      }})
    Scale Alpha=0.4
   ) {enabled=false}
   &ScaleColor(
    Scale RGB=&MakeVectorFromLinearColorRGB(Color=&ColorFromCurve(ColorCurve=DataInterface() {propertyValues {
        AlphaCurve {keys [<at 0.1151 1>,<at 0.9977 1 RCIM_Cubic>]}
        BlueCurve {keys [<at 0.1176 1>,<at 0.3047 1>,<at 0.5191 1>,<at 0.9868 0.5469>]}
        GreenCurve {keys [<at 0.1176 0.5>,<at 0.3047 0.5>,<at 0.5191 0>,<at 0.9868 0>]}
        RedCurve {keys [<at 0.1176 0.6355>,<at 0.3047 0.8034>,<at 0.5191 0.0453>,<at 0.9868 0.0593>]}
       }}))
    ScaleA=true
    ScaleRGB=true
   )
   &ScaleRibbonWidth(Ribbon Width Scale=&FloatFromCurve(
     FloatCurve=DataInterface() {propertyValues {Curve {keys [
         {interpMode=RCIM_Cubic, tangentMode=RCTM_User, time=0, value=1}
         {arriveTangent=11.4975, interpMode=RCIM_Cubic, leaveTangent=11.4975, tangentMode=RCTM_User, time=0.5, value=10}
        ]}}}
     CurveIndex=$Emitter.Age
    ))
   &DynamicMaterialParameters(
    Index 0 Param 1=&FloatFromCurve(FloatCurve=DataInterface() {propertyValues {Curve {keys [
         {arriveTangent=0.2015, interpMode=RCIM_Cubic, leaveTangent=0.2015, tangentMode=RCTM_User, time=0, value=1.5}
         {arriveTangent=-1.5941, interpMode=RCIM_Cubic, leaveTangent=-1.5941, tangentMode=RCTM_User, time=0.3, value=1}
         {interpMode=RCIM_Cubic, tangentMode=RCTM_User, time=1, value=0}
        ]}}})
    Index 0 Param 3=-0.8
    Index 0 Param 4=&FloatFromCurve(
     FloatCurve=DataInterface() {propertyValues {Curve {keys [
         {interpMode=RCIM_Cubic, tangentMode=RCTM_User, time=0, value=1}
         {arriveTangent=-0.6219, interpMode=RCIM_Cubic, leaveTangent=-0.6219, tangentMode=RCTM_User, time=1, value=0}
        ]}}}
     Scale Curve=0.1
    )
    Module.Index0Param0Write=true
    Module.Index0Param1Write=false
    Module.Index0Param2Write=true
    Module.Index0Param3Write=true
   )
  ]
  rendererClasses [&NiagaraRibbonRendererProperties]
  renderers [
   {rendererClass=&NiagaraRibbonRendererProperties}
   {rendererClass=&NiagaraRibbonRendererProperties, rendererIndex=1}
  ]
 }
}
summary {
 systemName=NS_Pickaxe_CosmicSquatter_v02_Trail
 emitters [
  {emitterName=SmokeTrail, rendererClasses [&NiagaraRibbonRendererProperties]}
  {emitterName=NebulaTrail, rendererClasses [&NiagaraRibbonRendererProperties]}
 ]
 userVariables [
  User.bIsFrontEnd {type=NiagaraBool}
  User.bIsFrontEndPreview {type=NiagaraBool}
  User.ColorCustomA {type=LinearColor, description="", defaultValue {struct=LinearColor, value=<rgba 1 1 1 1>}}
  User.ColorCustomB {type=LinearColor, description="", defaultValue {struct=LinearColor, value=<rgba 1 1 1 1>}}
  User.ColorSwitcher {type=NiagaraFloat, defaultValue=0, description=""}
  User.RibbonIDOffset {type=NiagaraInt32}
  User.SkeletalMesh {
   type=NiagaraDataInterfaceSkeletalMesh
   defaultValue {
    struct=NiagaraExt_VariableValue_DataInterface
    value {dataInterface=&self:NiagaraDataInterfaceSkeletalMesh_0}
   }
  }
  User.VariantPicker {type=NiagaraFloat, defaultValue=0, description=""}
 ]
}
