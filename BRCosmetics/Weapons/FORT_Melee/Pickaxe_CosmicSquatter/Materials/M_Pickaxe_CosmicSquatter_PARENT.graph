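# Parent material for the CosmicSquatter pickaxe, adding a galaxy effect and glowing rainbow edges to the 2023 character-parent shading.
# MFC_GalaxyV2 sits between the PreFX and PostFX functions, driven by channels of a packed tech-art mask.
# A scrolling, pulsing spectrum gradient plus screen-and-UV sparkle forms the edge color, dimmed at night.
# That color blends into Emissive through masks, including a vertical fade derived from pre-skinned position.
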Material /BRCosmetics/Weapons/FORT_Melee/Pickaxe_CosmicSquatter/Materials/M_Pickaxe_CosmicSquatter_PARENT
graph {
 properties {
  bUseMaterialAttributes=true
  bUsedWithMeshDeformer=true
  bUsedWithNanite=true
  bUsedWithSkeletalMesh=true
  pixelDepthOffsetMode=PDOM_Legacy
  shadingModel=MSM_FromMaterialExpression
  usedShadingModels=Default Lit | Subsurface | Clear Coat | Subsurface Profile | Cloth
 }
 nodes [
  n1 = VertexInterpolator(VS=PreSkinnedPosition())
  n2 = ScalarParameter() {defaultValue=0.1, group=CosmicSquatter, parameterName=EdgeColorFade_Hardness}
  n3 = Add(
   A=Multiply(A=ComponentMask(None=@n1.PS) {b=true}, B=@n2) {constB=0.1}
   B=Multiply(A=@n2, B=ScalarParameter() {defaultValue=-40, group=CosmicSquatter, parameterName=EdgeColorFade_Offset})
  ) {constB=-40}
  NamedRerouteDeclaration(None=OneMinus(None=Saturate(None=@n3))) {name=VerticalFadeMask, nodeColor=<rgba 0.9583 1 1 1>}
  n4 = TextureSampleParameter2D() {
   group=05_Tech Art Masks
   parameterName=TechArtMask
   texture=/BRCosmetics/Weapons/FORT_Melee/Pickaxe_CosmicSquatter/Textures/T_Pickaxe_CosmicSquatter_FX
  }
  Reroute(None=@n4.R)
  n5 = MaterialFunctionCall() {materialFunction=&MF_CharacterParent_2023_PreFX}
  n6 = Add(A=@n4.B, B=NamedRerouteUsage())
  n7 = Multiply(A=@n4.G) {constB=3}
  n8 = MaterialFunctionCall(R=NamedRerouteDeclaration(None=@n4.R) {name=EdgeMask, nodeColor=<rgba 1 0.2465 0.3162 1>}, G=@n4.G, B=Saturate(None=Subtract(A=@n6, B=@n7))) {materialFunction=&AppendMany}
  n9 = GetMaterialAttributes(None=MaterialFunctionCall(
    Material=MaterialFunctionCall!(MA=@n5.Result, SkinFX=@n8.RGB) {materialFunction=/Game/Characters/Player/Female/Medium/Bodies/F_MED_StarStray/Skins/Galaxy/Materials/MFC_GalaxyV2}.Result
    M=@n5.M
    Pre Skinned Local Position=@n5.M
    Distance Blend=@n5.M
    Fresnel=@n5.M
   ) {materialFunction=&MF_CharacterParent_2023_PostFX}.Result) {attributeGetTypes [#1]}
  n10 = Time()
  n11 = ScalarParameter() {defaultValue=1, group=CosmicSquatter, parameterName=PulseSpeed, sortPriority=1}
  n12 = Sine(None=Subtract(A=NamedRerouteUsage(), B=Multiply(A=@n10, B=@n11)))
  n13 = Panner(
   Coordinate=Add(A=NamedRerouteUsage(), B=ScalarParameter() {group=CosmicSquatter, parameterName=ColorfulTextureOffset, sortPriority=0})
   Speed=AppendVector(A=ScalarParameter() {defaultValue=-0.5, group=CosmicSquatter, parameterName=ScrollingSpeed, sortPriority=1}, B=Constant())
  ) {speedX=-0.5}
  n14 = TextureSampleParameter2D(UVs=Multiply(A=Frac(None=@n13), B=ScalarParameter() {defaultValue=1, group=CosmicSquatter, parameterName=ColorfulTextureTiling, sortPriority=1})) {
   desc=Gradient that defines the rainbow color
   group=CosmicSquatter
   parameterName=ColorfulEdgeTexture
   sortPriority=0
   texture=/Game/Global/Textures/Default/Thinfilm/T_ThinFilm_Spectrum_COLOR
  }
  n15 = Multiply(A=Add(A=Saturate(None=@n12)), B=@n14.RGB)
  n16 = Multiply(A=ScreenPosition().ViewportUV) {constB=2}
  n17 = Multiply(A=TextureCoordinate()) {constB=2.5}
  n18 = Multiply(
   A=Power(Base=TextureSample(UVs=@n16) {samplerType=SAMPLERTYPE_Masks, texture=/Game/Gadgets/Assets/VinderTech_GliderChute/Textures/T_Gumdrop_Sparkle_M}.R) {constExponent=8}
   B=Power(Base=TextureSample(UVs=@n17) {samplerType=SAMPLERTYPE_Masks, texture=&T_Gumdrop_Sparkle_M}.RGB)
  )
  n19 = Desaturation(
   None=Multiply(A=Add(A=@n15, B=@n18), B=ScalarParameter() {defaultValue=1, group=CosmicSquatter, parameterName=EdgeColorBrightness})
   Fraction=ScalarParameter() {group=CosmicSquatter, parameterName=EdgeColorDesaturation}
  )
  n20 = Constant() {r=1}
  n21 = MaterialFunctionCall(Value_Morning=@n20, Value_Day=@n20, Value_Evening=@n20, Value_Night=Constant() {r=0.5}) {materialFunction=&Scalar_Time_of_Day}
  n22 = Add(A=Add(A=@n4.B, B=@n4.B), B=NamedRerouteUsage())
  SetMaterialAttributes(
   MaterialAttributes=@n9.MaterialAttributes
   Emissive Color=LinearInterpolate(A=Multiply(A=@n19, B=@n21.Result), B=@n9.Emissive Color, Alpha=Saturate(None=@n22))
  ) {attributeSetTypes [#1]}
 ]
}
