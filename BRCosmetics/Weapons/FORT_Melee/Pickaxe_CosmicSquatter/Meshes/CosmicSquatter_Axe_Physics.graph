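# Physics asset for the CosmicSquatter pickaxe, simulating two bandana strands, top and bottom, attached to the two-handed handle.
# A long capsule covers the handle, and each strand has a small master body followed by a chain of four capsule bodies.
# Master bodies join the handle with free swing and twist, while chain links limit swing, widest at the first link, and keep twist tight.
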PhysicsAsset /BRCosmetics/Weapons/FORT_Melee/Pickaxe_CosmicSquatter/Meshes/CosmicSquatter_Axe_Physics
bodies {
 MSTR_bandana_Btm [{shapeType=Capsule, length=0.505, radius=0.505, shapeName=MSTR_bandana_Btm_capsule, rotation=<rot 0 -90 0>}]
 MSTR_bandana_Top [{shapeType=Capsule, length=0.505, radius=0.505, shapeName=MSTR_bandana_Top_capsule, rotation=<rot 0 -90 0>}]
 Melee_TwoHanded_Handle [{shapeType=Capsule, length=81.7561, radius=2.1131, shapeName=Melee_TwoHanded_Handle_capsule, center=<0 -0.6359 0>}]
 dyn_bandana_Btm_00 [{shapeType=Capsule, length=3.5, radius=0.85, shapeName=dyn_bandana_Btm_00_capsule, center=<5.2163 0 0>, rotation=<rot 90 0 0>}]
 dyn_bandana_Btm_01 [{shapeType=Capsule, length=7.5, radius=1, shapeName=dyn_bandana_Btm_01_capsule, center=<4.2305 0 0>, rotation=<rot 90 0 0>}]
 dyn_bandana_Btm_02 [{
   shapeType=Capsule
   length=4.184
   radius=1.576
   shapeName=dyn_bandana_Btm_02_capsule
   center=<4.1015 -0.0156 0.3392>
   rotation=<rot 0 89.7309 -85.9521>
  }]
 dyn_bandana_Btm_03 [{shapeType=Capsule, length=4, radius=2, shapeName=dyn_bandana_Btm_03_capsule, center=<1.6156 0 0>, rotation=<rot 90 0 0>}]
 dyn_bandana_Top_00 [{shapeType=Capsule, length=5.5, radius=0.5, shapeName=dyn_bandana_Top_00_capsule, center=<5.0515 0 0>, rotation=<rot 90 0 0>}]
 dyn_bandana_Top_01 [{shapeType=Capsule, length=7.5, radius=1, shapeName=dyn_bandana_Top_01_capsule, center=<4.2305 0 0>, rotation=<rot 90 0 0>}]
 dyn_bandana_Top_02 [{
   shapeType=Capsule
   length=5.1017
   radius=1.5502
   shapeName=dyn_bandana_Top_02_capsule
   center=<5.0029 0.2267 0.1638>
   rotation=<rot 0 90.0414 -87.6606>
  }]
 dyn_bandana_Top_03 [{shapeType=Capsule, length=4, radius=2, shapeName=dyn_bandana_Top_03_capsule, center=<1.6156 0 0>, rotation=<rot 90 0 0>}]
}
constraints [
 {bone1Name=dyn_bandana_Btm_00, bone2Name=MSTR_bandana_Btm, swing1LimitDegrees=60, swing2LimitDegrees=60, twistLimitDegrees=5}
 {bone1Name=dyn_bandana_Btm_01, bone2Name=dyn_bandana_Btm_00, swing1LimitDegrees=45, swing2LimitDegrees=45, twistLimitDegrees=5}
 {bone1Name=dyn_bandana_Btm_02, bone2Name=dyn_bandana_Btm_01, swing1LimitDegrees=45, swing2LimitDegrees=45, twistLimitDegrees=5}
 {bone1Name=dyn_bandana_Top_00, bone2Name=MSTR_bandana_Top, swing1LimitDegrees=60, swing2LimitDegrees=60, twistLimitDegrees=5}
 {bone1Name=dyn_bandana_Top_01, bone2Name=dyn_bandana_Top_00, swing1LimitDegrees=45, swing2LimitDegrees=45, twistLimitDegrees=5}
 {bone1Name=dyn_bandana_Top_02, bone2Name=dyn_bandana_Top_01, swing1LimitDegrees=45, swing2LimitDegrees=45, twistLimitDegrees=5}
 {
  bone1Name=MSTR_bandana_Top
  bone2Name=Melee_TwoHanded_Handle
  swing1LimitDegrees=45
  swing1Motion=Free
  swing2LimitDegrees=45
  swing2Motion=Free
  twistLimitDegrees=45
  twistMotion=Free
 }
 {
  bone1Name=MSTR_bandana_Btm
  bone2Name=Melee_TwoHanded_Handle
  swing1LimitDegrees=45
  swing1Motion=Free
  swing2LimitDegrees=45
  swing2Motion=Free
  twistLimitDegrees=45
  twistMotion=Free
 }
 {bone1Name=dyn_bandana_Btm_03, bone2Name=dyn_bandana_Btm_02, swing1LimitDegrees=45, swing2LimitDegrees=45, twistLimitDegrees=5}
 {bone1Name=dyn_bandana_Top_03, bone2Name=dyn_bandana_Top_02, swing1LimitDegrees=45, swing2LimitDegrees=45, twistLimitDegrees=5}
]
